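# Niagara ribbon trail effect for the RoseJewel pickaxe, built on the FN anim-trail modules that follow bone or socket pairs.
# TrailBlade runs two trails with color and lifetime split by spawn group, TrailBarrel covers another socket pair, and TrailFlame derives top and bottom flame colors from curves over the particle color.
# All three emitters use system-level spawn fade, system fade and time-of-day tint; blade and barrel also darken RGB to black over their short lifetimes.
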NiagaraSystem /BRCosmetics/Weapons/FORT_Melee/Pickaxe_RoseJewel/FX/NS_Pickaxe_RoseJewel_Trail
dependencies {
 usedDataInterfaces [&NiagaraDataInterfaceColorCurve, &NiagaraDataInterfaceSocketReader, &NiagaraDataInterfaceCurve]
 usedDynamicInputs [
  &MergeHelper_FloatSwitch
  &FNTimeOfDayColor
  &FloatFromCurve
 ]
 usedModules [
  &self:SystemSpawnScript.SystemScriptSource.SystemScriptGraph.NiagaraNodeAssignment_1.SetVariables_#1
  &self:SystemSpawnScript.SystemScriptSource.SystemScriptGraph.NiagaraNodeAssignment_3.SetVariables_#2
  &SystemState
  &self:SystemSpawnScript.SystemScriptSource.SystemScriptGraph.NiagaraNodeAssignment_4.SetVariables_#3
  &self:SystemSpawnScript.SystemScriptSource.SystemScriptGraph.NiagaraNodeAssignment_7.SetVariables_#4
 ]
}
emitters {
 TrailBarrel {
  emitterUpdateScript [
   &EmitterState()
   &FN_SpawnAnimTrail(Number of Anim Trails=2)
  ]
  particleSpawnScript [
   &FN_InitializeAnimTrailBoneIndecies(First Bone or Socket Index=4, Second Bone or Socket Index=3)
   &FN_InitializeAnimTrail(
    Color=&MakeLinearColorFromVectorAndFloat(
     "Vector (RGB)"=&MakeVectorFromLinearColorRGB(Color=<rgba 0.2704 0.274 0.3391 1>)
     "Float (Alpha)"=&Multiply_Float(A=0.8, B=$System.FadeSpawn)
    )
    Lifetime=0.2
   )
  ]
  particleUpdateScript [
   &FN_UpdateAnimTrail() {module=FN_UpdateAnimTrail001}
   &ScaleColor(
    Scale Mode=1
    Scale RGBA=&Multiply_Vector4(
     A=&MakeVector4_FromLinearColor(LinearColor=$System.TODColor)
     B=&MakeVector4FromVectorAndFloat("Vector (RGB)"=<1 1 1>, "Float (Alpha)"=$System.FadeSystem)
    )
    ScaleRGBA=true
   )
   &ScaleColor(
    Linear Color Curve=DataInterface() {propertyValues {
      AlphaCurve {keys [<at 0 1>]}
      BlueCurve {keys [<at 0 1>,<at 0.9436 0>]}
      GreenCurve {keys [<at 0 1>,<at 0.9436 0>]}
      RedCurve {keys [<at 0 1>,<at 0.9436 0>]}
     }}
    Scale Mode=2
   ) {module=ScaleColor001}
  ]
  rendererClasses [&NiagaraRibbonRendererProperties]
  renderers [{rendererClass=&NiagaraRibbonRendererProperties}]
 }
 TrailBlade {
  emitterUpdateScript [
   &EmitterState()
   &FN_SpawnAnimTrail(Number of Anim Trails=2)
  ]
  particleSpawnScript [
   &FN_InitializeAnimTrailBoneIndecies()
   &FN_InitializeAnimTrailBoneIndecies(AnimTrail Selection=1, Second Bone or Socket Index=2) {module=FN_InitializeAnimTrailBoneIndecies001}
   &FN_InitializeAnimTrail(
    Color=&MakeLinearColorFromVectorAndFloat(
     "Vector (RGB)"=&MakeVectorFromLinearColorRGB(Color=&MaskLinearColorBySpawnGroup(Masked Linear Color=<rgba 0.0438 0.0438 0.08 1>, Passthrough Linear Color=<rgba 0.2783 0 0.0093 1>))
     "Float (Alpha)"=&Multiply_Float(A=2, B=$System.FadeSpawn)
    )
    Lifetime=&MaskFloatBySpawnGroup(Masked Float=0.16, Passthrough Float=0.1)
   )
  ]
  particleUpdateScript [
   &FN_UpdateAnimTrail() {module=FN_UpdateAnimTrail001}
   &ScaleColor(
    Scale Mode=1
    Scale RGBA=&Multiply_Vector4(
     A=&MakeVector4_FromLinearColor(LinearColor=$System.TODColor)
     B=&MakeVector4FromVectorAndFloat("Vector (RGB)"=<1 1 1>, "Float (Alpha)"=$System.FadeSystem)
    )
    ScaleRGBA=true
   )
   &ScaleColor(
    Linear Color Curve=DataInterface() {propertyValues {
      AlphaCurve {keys [<at 0 1>]}
      BlueCurve {keys [<at 0 1>,<at 0.9436 0>]}
      GreenCurve {keys [<at 0 1>,<at 0.9436 0>]}
      RedCurve {keys [<at 0 1>,<at 0.9436 0>]}
     }}
    Scale Mode=2
   ) {module=ScaleColor001}
  ]
  rendererClasses [&NiagaraRibbonRendererProperties]
  renderers [{rendererClass=&NiagaraRibbonRendererProperties}]
 }
 TrailFlame {
  emitterUpdateScript [
   &EmitterState()
   &FN_SpawnAnimTrail()
  ]
  particleSpawnScript [
   &FN_InitializeAnimTrailBoneIndecies()
   &FN_InitializeAnimTrail(
    Color=&MakeLinearColorFromVectorAndFloat(
     "Vector (RGB)"=&MakeVectorFromLinearColorRGB(Color=<rgba 1 1 1 1>)
     "Float (Alpha)"=&Multiply_Float(A=0.7, B=$System.FadeSpawn)
    )
    Lifetime=0.16
   )
  ]
  particleUpdateScript [
   &FN_UpdateAnimTrail() {module=FN_UpdateAnimTrail001}
   &ScaleColor(
    Scale Mode=1
    Scale RGBA=&Multiply_Vector4(
     A=&MakeVector4_FromLinearColor(LinearColor=$System.TODColor)
     B=&MakeVector4FromVectorAndFloat("Vector (RGB)"=<1 1 1>, "Float (Alpha)"=$System.FadeSystem)
    )
    ScaleRGBA=true
   )
   /SetVariables(
    Particles.Flame Color Bottom:LinearColor=&MultiplyLinearColor(
     Linear Color=$Particles.Color
     Scale Factor=&MakeVector4_FromLinearColor(LinearColor=&ColorFromCurve(ColorCurve=DataInterface() {propertyValues {
         AlphaCurve {keys [<at 0.1005 0.652>,<at 0.2509 1.2>,<at 0.6291 0.011>]}
         BlueCurve {keys [<at 0.0808 0.0187>,<at 0.3634 0.0057>]}
         GreenCurve {keys [<at 0.0808 0.02>,<at 0.3634 0.0057>]}
         RedCurve {keys [<at 0.0808 0.026>,<at 0.3634 0.3281>]}
        }}))
    )
    Particles.Flame Color Top:LinearColor=&MultiplyLinearColor(
     Linear Color=$Particles.Color
     Scale Factor=&MakeVector4_FromLinearColor(LinearColor=&ColorFromCurve(ColorCurve=DataInterface() {propertyValues {
         AlphaCurve {keys [<at 0.0344 1.4>,<at 0.3339 0>]}
         BlueCurve {keys [<at 0.0401 0.55>,<at 0.1961 0.1103>]}
         GreenCurve {keys [<at 0.0401 0.5659>,<at 0.1961 0.121>]}
         RedCurve {keys [<at 0.0401 0.6146>,<at 0.1961 0.1719>]}
        }}))
    )
   ) {isSetParameters=true}
  ]
  rendererClasses [&NiagaraRibbonRendererProperties]
  renderers [{rendererClass=&NiagaraRibbonRendererProperties}]
 }
}
summary {
 systemName=NS_Pickaxe_RoseJewel_Trail
 emitters [
  {emitterName=TrailBlade, rendererClasses [&NiagaraRibbonRendererProperties]}
  {emitterName=TrailFlame, rendererClasses [&NiagaraRibbonRendererProperties]}
  {emitterName=TrailBarrel, rendererClasses [&NiagaraRibbonRendererProperties]}
 ]
 userVariables [
  User.bIsFrontEnd {type=NiagaraBool}
  User.bIsFrontEndPreview {type=NiagaraBool}
  User.RibbonIDOffset {type=NiagaraInt32}
  User.SwingSectionIndex {type=NiagaraInt32}
 ]
}
